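# MetaSound source preset that plays the default first-person magazine eject sound for weapon reloads, as a stereo one-shot.
# It inherits its graph from MSS_FPS_Weapon_Reload and supplies four standard magazine eject waves, plus output volume, pan and narrow random pitch and volume ranges.
# Playback uses the first-person weapon foley attenuation, sound class and concurrency settings.
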MetaSoundSource /FortMetaSoundWeapon/Items/Metasounds/FPS/Presets/MSS_Reloads_FP_Mag_Eject_Default
preset_parent=/FortMetaSoundWeapon/Items/Metasounds/FPS/MSS_FPS_Weapon_Reload
inputs [
 OutputVolume {type=Float}
 PanAmount {type=Float}
 Random.PitchShift.Max {type=Float}
 Random.PitchShift.Min {type=Float}
 Random.Volume.Max {type=Float}
 Random.Volume.Min {type=Float}
 Sounds {type=WaveAsset:Array}
 UE.Source.OnPlay {type=Trigger}
]
interfaces [UE.OutputFormat.Stereo:1.0, UE.Source.OneShot:1.0, UE.Source:1.1]
outputs [
 "UE.OutputFormat.Stereo.Audio:0" {type=Audio}
 "UE.OutputFormat.Stereo.Audio:1" {type=Audio}
 UE.Source.OneShot.OnFinished {type=Trigger}
]
pages [Default {input_defaults {
   OutputVolume=0.301
   PanAmount=0.5
   Random.PitchShift.Max=1.05
   Random.PitchShift.Min=0.95
   Random.Volume.Max=1.05
   Random.Volume.Min=0.95
   UE.Source.OnPlay=null
   Sounds [
    /Game/Sounds/Fort_Weapons/Magazines/Magazine_Standard_Eject_01
    /Game/Sounds/Fort_Weapons/Magazines/Magazine_Standard_Eject_02
    /Game/Sounds/Fort_Weapons/Magazines/Magazine_Standard_Eject_03
    /Game/Sounds/Fort_Weapons/Magazines/Magazine_Standard_Eject_04
   ]
  }}]
properties {
 assetImportData=&self:AssetImportData
 attenuationSettings=/FortMetaSoundWeapon/Items/Shared/Settings/FPS/AS_Weapon_FP_Foley
 maxDistance=2250
 numChannels=2
 outputFormat=Stereo
 sampleRate=48000
 soundClassObject=&Fort_WeaponFoley
 concurrencySet [&CS_Fort_WeaponFoley]
}
